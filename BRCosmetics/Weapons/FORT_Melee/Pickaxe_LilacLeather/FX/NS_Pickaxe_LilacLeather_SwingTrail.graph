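# Niagara swing trail effect for the LilacLeather pickaxe cosmetic, drawing ribbons and sparks from weapon sockets during a swing.
# MainCable, Tip, Accents and SoftTrail are animation-trail ribbon emitters bound to socket indices, with Accents taking seven from a system array and SoftTrail adding a blue glow.
# DirectionalSparks spawns blue sprites at a socket, flung against its sampled velocity with drag and curl noise.
# Colors are scaled by a system TOD value.
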NiagaraSystem /BRCosmetics/Weapons/FORT_Melee/Pickaxe_LilacLeather/FX/NS_Pickaxe_LilacLeather_SwingTrail
dependencies {
 usedDataInterfaces [
  &NiagaraDataInterfaceCurve
  &NiagaraDataInterfaceColorCurve
  &NiagaraDataInterfaceSocketReader
  &NiagaraDataInterfaceArrayInt32
 ]
 usedDynamicInputs [&FN_FrontEndFloatModifier, &Vector2DFromFloat]
 usedModules [
  &self:SystemSpawnScript.SystemScriptSource.SystemScriptGraph.NiagaraNodeAssignment_7.SetVariables_#1
  &self:SystemSpawnScript.SystemScriptSource.SystemScriptGraph.NiagaraNodeAssignment_9.SetVariables_#2
  &self:SystemSpawnScript.SystemScriptSource.SystemScriptGraph.NiagaraNodeAssignment_12.SetVariables_#3
  &SystemState
  &self:SystemSpawnScript.SystemScriptSource.SystemScriptGraph.NiagaraNodeAssignment_6.SetVariables_#4
 ]
}
emitters {
 Accents {
  emitterSpawnScript [&FN_RibbonIDOffsetUtility()]
  emitterUpdateScript [
   &EmitterState(Life Cycle Mode=::NewEnumerator1, Loop Behavior=::NewEnumerator1, Loop Duration=0.3, UseLoopDelay=false)
   &FN_SpawnAnimTrail(Number of Anim Trails=7)
  ]
  particleSpawnScript [
   &FN_InitializeAnimTrailBoneIndecies(
    First Bone or Socket Index:NiagaraInt32=&SelectIntFromArray(Direct Array Index=0, Int Selection Array=$System.AccentSocketIndices)
    Second Bone or Socket Index:NiagaraInt32=&SelectIntFromArray(Direct Array Index=0, Int Selection Array=$System.AccentSocketIndices)
   )
   &FN_InitializeAnimTrailBoneIndecies(
    AnimTrail Selection=1
    First Bone or Socket Index:NiagaraInt32=&SelectIntFromArray(Direct Array Index=1, Int Selection Array=$System.AccentSocketIndices)
    Second Bone or Socket Index:NiagaraInt32=&SelectIntFromArray(Direct Array Index=1, Int Selection Array=$System.AccentSocketIndices)
   ) {module=FN_InitializeAnimTrailBoneIndecies006}
   &FN_InitializeAnimTrailBoneIndecies(
    AnimTrail Selection=2
    First Bone or Socket Index:NiagaraInt32=&SelectIntFromArray(Direct Array Index=2, Int Selection Array=$System.AccentSocketIndices)
    Second Bone or Socket Index:NiagaraInt32=&SelectIntFromArray(Direct Array Index=2, Int Selection Array=$System.AccentSocketIndices)
   ) {module=FN_InitializeAnimTrailBoneIndecies005}
   &FN_InitializeAnimTrailBoneIndecies(
    AnimTrail Selection=3
    First Bone or Socket Index:NiagaraInt32=&SelectIntFromArray(Direct Array Index=3, Int Selection Array=$System.AccentSocketIndices)
    Second Bone or Socket Index:NiagaraInt32=&SelectIntFromArray(Direct Array Index=3, Int Selection Array=$System.AccentSocketIndices)
   ) {module=FN_InitializeAnimTrailBoneIndecies004}
   &FN_InitializeAnimTrailBoneIndecies(
    AnimTrail Selection=4
    First Bone or Socket Index:NiagaraInt32=&SelectIntFromArray(Direct Array Index=4, Int Selection Array=$System.AccentSocketIndices)
    Second Bone or Socket Index:NiagaraInt32=&SelectIntFromArray(Direct Array Index=4, Int Selection Array=$System.AccentSocketIndices)
   ) {module=FN_InitializeAnimTrailBoneIndecies003}
   &FN_InitializeAnimTrailBoneIndecies(
    AnimTrail Selection=5
    First Bone or Socket Index:NiagaraInt32=&SelectIntFromArray(Direct Array Index=5, Int Selection Array=$System.AccentSocketIndices)
    Second Bone or Socket Index:NiagaraInt32=&SelectIntFromArray(Direct Array Index=5, Int Selection Array=$System.AccentSocketIndices)
   ) {module=FN_InitializeAnimTrailBoneIndecies002}
   &FN_InitializeAnimTrailBoneIndecies(
    AnimTrail Selection=6
    First Bone or Socket Index:NiagaraInt32=&SelectIntFromArray(Direct Array Index=6, Int Selection Array=$System.AccentSocketIndices)
    Second Bone or Socket Index:NiagaraInt32=&SelectIntFromArray(Direct Array Index=6, Int Selection Array=$System.AccentSocketIndices)
   ) {module=FN_InitializeAnimTrailBoneIndecies001}
   &FN_InitializeAnimTrail(
    Color=&MultiplyLinearColor(
     Linear Color=<rgba 0.7812 0.7004 0.6922 1.2>
     Scale Factor=&MakeVector4_FromLinearColor(LinearColor=$System.TOD)
    )
    Lifetime=0.12
   )
   /SetVariables(Particles.RibbonWidth=8) {isSetParameters=true, module=SetVariables_#5}
  ]
  particleUpdateScript [
   &FN_UpdateAnimTrail(Locked Particle RibbonWidth:NiagaraFloat=$Particles.RibbonWidth)
   &ScaleColor(
    Scale Alpha=&Multiply_Float(
     A=&FloatFromCurve(
      FloatCurve=DataInterface() {propertyValues {Curve {keys [<at 0 1 RCIM_Cubic>, {arriveTangent=-1, interpMode=RCIM_Cubic, leaveTangent=-1, time=0.4, value=0.5}, <at 1 0 RCIM_Cubic>]}}}
      Scale Curve=$System.SwingGeneralFade
     )
     B=0.1
    )
    Scale RGB=<0.8 0.8 0.8>
    ScaleA=true
    ScaleRGB=true
   )
   /SetVariables(Particles.RibbonWidth=&Multiply_Float(
     A=&FloatFromCurve(FloatCurve=DataInterface() {propertyValues {Curve {keys [<at 0.05 1 RCIM_Cubic>,<at 1 0 RCIM_Cubic>]}}})
     B=$Particles.Initial.RibbonWidth
    )) {isSetParameters=true}
  ]
  rendererClasses [&NiagaraRibbonRendererProperties]
  renderers [{rendererClass=&NiagaraRibbonRendererProperties}]
 }
 DirectionalSparks {
  emitterUpdateScript [
   &EmitterState()
   &SpawnRate(SpawnRate=40, Use Spawn Probability=false)
  ]
  particleSpawnScript [
   &InitializeParticle(
    Color=&MultiplyLinearColor(Linear Color=<rgba 0 1.1702 2 1>, Scale Factor=&MakeVector4_FromLinearColor(LinearColor=$System.TOD))
    Lifetime=&FloatFromCurve(
     FloatCurve=DataInterface() {propertyValues {Curve {keys [<at 0 0.12>,<at 0.85 0.15>,<at 0.92 0.4>,<at 1 0.5>]}}}
     CurveIndex=&RandomRangeFloat(Minimum=0, Maximum=1)
    )
    Sprite Size=&Multiply_Vector2D_ByFloat(Float=&RandomRangeFloat(Minimum=3, Maximum=5), Vector2D=<1 1.5>)
    Sprite Size Mode=::NewEnumerator4
    UsePositionOffset=false
    Write Material Random=true
   )
   &SocketLocation(
    Interpolate Socket Sampling=true
    SamplePosOffset=false
    Socket Index=21
    Socket Reader=$System.Socket Reader
    Socket Sampling Mode=::NewEnumerator3
    VelocitySpeedLimit=false
   )
   &ShapeLocation(
    Cylinder Height=20
    Cylinder Radius=2
    MaskSpawnGroup=false
    Override Seed=false
    Rotation Mode=::NewEnumerator1
    Shape Primitive=::NewEnumerator1
    Surface Only=false
    UseOwnerScale=true
   )
   &CameraOffset(Camera Offset Amount=20)
   &AddVelocity(
    Rotation Coordinate Space=::World
    Velocity=&Multiply_VectorByFloat(
     Vector=&NormalizeVector(Vector To Normalize=$Output.SocketLocation.SampledVelocity, Safe Normalize=true, Safe Fallback Threshold=0, Safe Fallback Vector=<0 0 0.001>)
     Float=-1
    )
    Velocity Speed Scale=&RandomRangeFloat(Minimum=100, Maximum=200)
   )
  ]
  particleUpdateScript [
   &ParticleState()
   &Drag(Drag=&RandomRangeFloat(Minimum=10, Maximum=10), Use Linear Drag=true, Use Rotational Drag=true)
   &CurlNoiseForce(Bias Noise=false, Noise Frequency=30, Noise Strength=500, Pan Noise=false)
   &SolveForcesAndVelocity(Clamp Velocity=false, Limit Acceleration=false)
   &ScaleSpriteSizeBySpeed(Max Scale Factor=<0.2 5>, Min Scale Factor=<0.2 0.2>)
   &ScaleSpriteSize(Uniform Curve Sprite Scale=DataInterface() {propertyValues {Curve {keys [{arriveTangent=-6.725, interpMode=RCIM_Cubic, leaveTangent=-6.725, tangentMode=RCTM_User, time=0.65, value=1}, <at 1 0 RCIM_Cubic>]}}})
   &ScaleColor(
    Scale RGB=&VectorFromFloat(Value=&FloatFromCurve(
      FloatCurve=DataInterface() {propertyValues {Curve {keys [
          {interpMode=RCIM_Cubic, tangentMode=RCTM_User, time=0, value=1}
          <at 0.3 0.2>
          {interpMode=RCIM_Cubic, tangentMode=RCTM_User, time=1, value=0}
         ]}}}
      Scale Curve=10
     ))
    ScaleA=true
    ScaleRGB=true
   )
   /SetVariables(
    Particles.GlowColor=&MakeLinearColorFromVectorAndFloat(
     "Vector (RGB)"=&MakeVectorFromLinearColorRGB(Color=<rgba 0 0.1937 1 1>)
     "Float (Alpha)"=&FloatFromCurve(FloatCurve=DataInterface() {propertyValues {Curve {keys [<at 0.5 1>,<at 1 0>]}}}, Scale Curve=0.1)
    )
    Particles.GlowSize=&Multiply_Vector2D(A=$Particles.SpriteSize, B=<8 2>)
   ) {isSetParameters=true}
  ]
  rendererClasses [&NiagaraSpriteRendererProperties]
  renderers [
   {rendererClass=&NiagaraSpriteRendererProperties}
   {rendererClass=&NiagaraSpriteRendererProperties, rendererIndex=1}
  ]
 }
 MainCable {
  emitterSpawnScript [&FN_RibbonIDOffsetUtility()]
  emitterUpdateScript [
   &EmitterState()
   &FN_SpawnAnimTrail()
  ]
  particleSpawnScript [
   &FN_InitializeAnimTrailBoneIndecies(First Bone or Socket Index=14)
   /SetVariables(Particles.RibbonWidth=100) {isSetParameters=true, module=SetVariables_#5}
   &FN_InitializeAnimTrail(
    Color=&MultiplyLinearColor(
     Linear Color=<rgba 0.0938 0.0929 0.0802 0.5>
     Scale Factor=&MakeVector4_FromLinearColor(LinearColor=$System.TOD)
    )
    Lifetime=0.15
   )
  ]
  particleUpdateScript [
   &FN_UpdateAnimTrail(Locked Particle RibbonWidth:NiagaraFloat=$Particles.RibbonWidth)
   &ScaleColor(
    Scale Alpha=&Multiply_Float(
     A=&FloatFromCurve(
      FloatCurve=DataInterface() {propertyValues {Curve {keys [<at 0 1 RCIM_Cubic>, {arriveTangent=-1, interpMode=RCIM_Cubic, leaveTangent=-1, time=0.4, value=0.375}, <at 1 0 RCIM_Cubic>]}}}
      Scale Curve=$System.SwingGeneralFade
     )
     B=0.35
    )
    ScaleA=true
    ScaleRGB=true
   )
  ]
  rendererClasses [&NiagaraRibbonRendererProperties]
  renderers [{rendererClass=&NiagaraRibbonRendererProperties}]
 }
 SoftTrail {
  emitterUpdateScript [
   &EmitterState(Life Cycle Mode=::NewEnumerator1, Loop Behavior=::NewEnumerator1, Loop Duration=0.42, UseLoopDelay=false)
   &FN_SpawnAnimTrail(Number of Anim Trails=3)
  ]
  particleSpawnScript [
   &FN_InitializeAnimTrailBoneIndecies(First Bone or Socket Index=15, Second Bone or Socket Index=16)
   &FN_InitializeAnimTrailBoneIndecies(AnimTrail Selection=1, First Bone or Socket Index=17, Second Bone or Socket Index=18) {module=FN_InitializeAnimTrailBoneIndecies001}
   &FN_InitializeAnimTrailBoneIndecies(AnimTrail Selection=1, First Bone or Socket Index=19, Second Bone or Socket Index=20) {module=FN_InitializeAnimTrailBoneIndecies002}
   &FN_InitializeAnimTrail(
    Color=&MultiplyLinearColor(Linear Color=<rgba 0.4 0.4 0.55 0.5>, Scale Factor=&MakeVector4_FromLinearColor(LinearColor=$System.TOD))
    Lifetime=0.25
   )
  ]
  particleUpdateScript [
   &FN_UpdateAnimTrail()
   &ScaleColor(
    Scale Alpha=&Multiply_Float(
     A=&FloatFromCurve(FloatCurve=DataInterface() {propertyValues {Curve {keys [
          {arriveTangent=-1.1204, interpMode=RCIM_Cubic, leaveTangent=-1.1204, tangentMode=RCTM_User, time=0, value=0}
          <at 0.1 1 RCIM_Cubic>
          {interpMode=RCIM_Cubic, tangentMode=RCTM_User, time=1, value=0}
         ]}}})
     B=&FloatFromCurve(
      FloatCurve=DataInterface() {propertyValues {Curve {keys [
          {arriveTangent=0.0483, interpMode=RCIM_Cubic, leaveTangent=0.0483, tangentMode=RCTM_User, time=0.5, value=1}
          {interpMode=RCIM_Cubic, tangentMode=RCTM_User, time=0.8, value=0}
         ]}}}
      CurveIndex=$Emitter.NormalizedLoopAge
     )
    )
    ScaleA=true
    ScaleRGB=true
   )
   /SetVariables(
    Particles.GlowColor=&MultiplyLinearColor(
     Linear Color=&MakeLinearColorFromVectorAndFloat(
      "Vector (RGB)"=&MakeVectorFromLinearColorRGB(Color=&ColorFromCurve(ColorCurve=DataInterface() {propertyValues {
          AlphaCurve {keys [<at 0 1>,<at 1 0>]}
          BlueCurve {keys [<at 0 1>, {arriveTangent=-1.2056, leaveTangent=-1.2056, time=0.5638, value=0.4479}, <at 0.9919 0.2812>]}
          GreenCurve {keys [<at 0 0.4632>, {arriveTangent=-1.3784, leaveTangent=-1.3784, time=0.5638, value=0.1163}, <at 0.9919 0.0806>]}
          RedCurve {keys [<at 0 0>, {arriveTangent=-0.2448, leaveTangent=-0.2448, time=0.5638, value=0.0534}, <at 0.9919 0.0832>]}
         }}))
      "Float (Alpha)"=&Multiply_Float(
       A=&FloatFromCurve(FloatCurve=DataInterface() {propertyValues {Curve {keys [<at 0 1 RCIM_Cubic>,<at 0.6 0 RCIM_Cubic>]}}}, Scale Curve=$System.SwingGeneralFade)
       B=1
      )
     )
     Scale Factor=&Multiply_Vector4(
      A=&MakeVector4_FromLinearColor(LinearColor=<rgba 1.25 1.25 1.25 2>)
      B=&MakeVector4_FromLinearColor(LinearColor=$System.TOD)
     )
    )
    Particles.GlowWidth:NiagaraFloat=&FloatFromCurve(
     FloatCurve=DataInterface() {propertyValues {Curve {keys [{arriveTangent=-2.5517, interpMode=RCIM_Cubic, leaveTangent=-2.5517, tangentMode=RCTM_User, time=0.2, value=1}, <at 1 0 RCIM_Cubic>]}}}
     Scale Curve=10
    )
   ) {isSetParameters=true}
  ]
  rendererClasses [&NiagaraRibbonRendererProperties]
  renderers [
   {rendererClass=&NiagaraRibbonRendererProperties}
   {rendererClass=&NiagaraRibbonRendererProperties, rendererIndex=1}
  ]
 }
 Tip {
  emitterUpdateScript [
   &EmitterState()
   &FN_SpawnAnimTrail(Number of Anim Trails=3)
  ]
  particleSpawnScript [
   &FN_InitializeAnimTrailBoneIndecies(First Bone or Socket Index=15, Second Bone or Socket Index=15)
   &FN_InitializeAnimTrailBoneIndecies(AnimTrail Selection=1, First Bone or Socket Index=18, Second Bone or Socket Index=18) {module=FN_InitializeAnimTrailBoneIndecies001}
   &FN_InitializeAnimTrailBoneIndecies(AnimTrail Selection=2, First Bone or Socket Index=19, Second Bone or Socket Index=20) {module=FN_InitializeAnimTrailBoneIndecies002}
   &FN_InitializeAnimTrail(
    Color=&MultiplyLinearColor(Linear Color=<rgba 1 1 1 0.5>, Scale Factor=&MakeVector4_FromLinearColor(LinearColor=$System.TOD))
    Lifetime=0.2
   )
  ]
  particleUpdateScript [
   &FN_UpdateAnimTrail()
   &ScaleColor(
    Scale Alpha=&Multiply_Float(
     A=&FloatFromCurve(
      FloatCurve=DataInterface() {propertyValues {Curve {keys [
          <at 0 1 RCIM_Cubic>
          {arriveTangent=-5.5556, interpMode=RCIM_Cubic, leaveTangent=-5.5556, time=0.05, value=0.5}
          {interpMode=RCIM_Cubic, leaveTangent=-2.5, time=0.18, value=0}
         ]}}}
      CurveIndex=$Particles.Age
     )
     B=1
    )
    Scale RGB=&MakeVectorFromLinearColorRGB(Color=&ColorFromCurve(ColorCurve=DataInterface() {propertyValues {
        AlphaCurve {keys [<at 0 1>,<at 1 0>]}
        BlueCurve {keys [
          <at 0.0052 5>
          <at 0.0336 2>
          <at 0.1008 1>
          <at 0.4234 1>
          {arriveTangent=-1.2056, leaveTangent=-1.2056, time=0.7792, value=0.6979}
          <at 1 0.2812>
         ]}
        GreenCurve {keys [
          <at 0.0052 2.5417>
          <at 0.0336 1.0884>
          <at 0.1008 0.5503>
          <at 0.4234 0.3193>
          {arriveTangent=-1.3784, leaveTangent=-1.3784, time=0.7792, value=0.1812}
          <at 1 0.0806>
         ]}
        RedCurve {keys [
          <at 0.0052 0>
          <at 0.0336 0.1458>
          <at 0.1008 0.1823>
          <at 0.4234 0.1354>
          {arriveTangent=-0.2448, leaveTangent=-0.2448, time=0.7792, value=0.0832}
          <at 1 0.0832>
         ]}
       }}))
    ScaleA=true
    ScaleRGB=true
   )
   /SetVariables(Particles.RibbonWidth=&FloatFromCurve(
     FloatCurve=DataInterface() {propertyValues {Curve {keys [{arriveTangent=-2.8662, interpMode=RCIM_Cubic, leaveTangent=-2.8662, tangentMode=RCTM_User, time=0.2, value=1}, <at 1 0 RCIM_Cubic>]}}}
     Scale Curve=2
    )) {isSetParameters=true}
  ]
  rendererClasses [&NiagaraRibbonRendererProperties]
  renderers [{rendererClass=&NiagaraRibbonRendererProperties}]
 }
}
summary {
 systemName=NS_Pickaxe_LilacLeather_SwingTrail
 emitters [
  {emitterName=MainCable, rendererClasses [&NiagaraRibbonRendererProperties]}
  {emitterName=Tip, rendererClasses [&NiagaraRibbonRendererProperties]}
  {emitterName=Accents, rendererClasses [&NiagaraRibbonRendererProperties]}
  {emitterName=SoftTrail, rendererClasses [&NiagaraRibbonRendererProperties]}
  {emitterName=DirectionalSparks, rendererClasses [&NiagaraSpriteRendererProperties]}
 ]
 userVariables [
  User.bIsFrontEnd {type=NiagaraBool}
  User.bIsFrontEndPreview {type=NiagaraBool}
  User.FrontEndScale {type=NiagaraFloat, defaultValue=0.6, description=""}
  User.RibbonIDOffset {type=NiagaraInt32}
 ]
}
